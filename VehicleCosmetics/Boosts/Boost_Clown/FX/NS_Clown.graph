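# NS_Clown is the Niagara effect for the Clown vehicle boost, emitting from sockets on the vehicle's skeletal mesh.
# Two near-identical Fire emitters spawn short-lived, paintable mesh and sprite particles that re-sample their socket every update.
# The HaHa emitter throws three spawn groups of flipbook sprites backward across the ports, slowing over life.
# Game-context masks scale spawn rates, while a frontend-preview flag switches HaHa size, color and velocity.
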NiagaraSystem /VehicleCosmetics/Boosts/Boost_Clown/FX/NS_Clown
dependencies {
 usedDataInterfaces [&NiagaraDataInterfaceCurve, &NiagaraDataInterfaceSkeletalMesh, &NiagaraDataInterfaceArrayFloat3]
 usedDynamicInputs [&SelectVectorFromArray]
 usedModules [
  &self:SystemSpawnScript.SystemScriptSource.SystemScriptGraph.NiagaraNodeAssignment_3.SetVariables_#1
  &self:SystemSpawnScript.SystemScriptSource.SystemScriptGraph.NiagaraNodeAssignment_4.SetVariables_#2
  &self:SystemSpawnScript.SystemScriptSource.SystemScriptGraph.NiagaraNodeAssignment_2.SetVariables_#3
  &SystemState
 ]
}
emitters {
 NE_VC_Boost_Sprite_Fire001 {
  emitterUpdateScript [
   &EmitterState(Inactive Response=::NewEnumerator1, Life Cycle Mode=::NewEnumerator1, UseLoopDelay=false)
   &VC_FilteredSocketCount()
   &SpawnRate(
    SpawnRate=&Multiply_Float(
     A=20
     B=/VehicleCosmetics/FX/Niagara/Modules/VC_Mask_Floats(Coro Value:NiagaraFloat=0.5, FrontEnd Value:NiagaraFloat=0.5, FN Value:NiagaraFloat=0.5)
    )
    Use Spawn Probability=false
   )
   &SpawnRate(
    SpawnGroup=1
    SpawnRate=&Multiply_Float(
     A=40
     B=&VC_Mask_Floats(Coro Value:NiagaraFloat=0.5, FrontEnd Value:NiagaraFloat=0.5, FN Value:NiagaraFloat=0.5)
    )
    Use Spawn Probability=false
   ) {module=SpawnRate001}
  ]
  particleSpawnScript [
   /SetVariables(Particles.VisibilityTag=&SpawnGroup()) {isSetParameters=true, module=SetVariables_#4}
   &VC_PaintedVariantsOverride()
   &InitializeParticle(
    Color Mode=::NewEnumerator0
    Lifetime=0.1
    Mesh Scale=&MergeHelper_VectorSwitch(
     0=&RandomRangeVector(Minimum=<0.8 0.8 0.8>, Maximum=<1 1 1>)
     1=&Lerp_Vector(
      A=&RandomRangeVector(Minimum=<1.2 0.6 1.2>, Maximum=<1.4 0.8 1.4>)
      B=&RandomRangeVector(Minimum=<0.4 1 0.4>, Maximum=<0.6 1.2 0.6>)
      Alpha=&MakeFloatFromInt(Int=&RandomRangeInt(Minimum=0, Maximum=1))
     )
     2=?
     3=?
     Index=&SpawnGroup()
     Num=::NewEnumerator0
    )
    Mesh Scale Mode=::NewEnumerator4
    UsePositionOffset=false
    Write Material Random=true
   )
   &PartitionParticles(Partition Mode:ENiagara_PartitionMode_v2=::NewEnumerator4)
   &SkeletalMeshLocation(
    Bone / Socket Index=$Particles.PartitionParticles.Partition
    Bone / Socket Sampling Mode=::NewEnumerator5
    MaskSpawnGroup=false
    Mesh Sampling Type=::NewEnumerator6
    Orientation Sampling=::NewEnumerator1
    Skeletal Mesh=$System.SkeletalMesh
   )
   &AddVelocity(
    Cone Angle=90
    Cone Axis=<-1 0 0>
    Module.UseConeFalloff=false
    Module.UseVelDistribution=false
    Override Seed=false
    Velocity Mode=::NewEnumerator2
    Velocity Speed=1
   ) {enabled=false}
   /VehicleCosmetics/FX/VC_FXParents/Modules/DLR_ProductPreviews(GameModeID:NiagaraInt32=$System.GameModeID, bPreviewVehicleSpeed:NiagaraBool=$System.bPreviewVehicleSpeed)
   &InitialMeshOrientation(
    Mesh Orientation Mode=::NewEnumerator7
    Rotation=&RandomRangeVector(Minimum=<0 0 0>, Maximum=<1 0 0>)
    Use Rotation Vector=true
   )
  ]
  particleUpdateScript [
   &ParticleState()
   &UpdateMeshOrientation(
    Facing Direction=<1 0 -0.1>
    Facing Interpolation=false
    Orientation Method=::NewEnumerator1
    Override Quat=false
    TransformFacing=false
    TransformUp=false
   )
   &InheritSourceMovement(
    Applied Position Scale=&Multiply_VectorByFloat(
     Vector=<1 1 1>
     Float=&FloatFromCurve(
      FloatCurve=DataInterface() {propertyValues {Curve {
         defaultValue=1
         keys [
          {arriveTangent=-2.2077, interpMode=RCIM_Cubic, leaveTangent=-2.2077, tangentMode=RCTM_User, time=0, value=1}
          {tangentMode=RCTM_User, time=1, value=0}
         ]
        }}}
      Scale Curve=0
     )
    )
    ApplyForce=false
    ApplyPosition=true
    ApplyPositionFalloff=false
   ) {enabled=false}
   &SolveForcesAndVelocity(Clamp Velocity=false, Limit Acceleration=false)
   &DynamicMaterialParameters(
    Index 1 Param 2=0.35
    Module.Index0Float0Write=true
    Module.Index0Float1Write=true
    Module.Index0Float2Write=true
    Module.Index0Float3Write=true
    Module.Index1Float0Write=true
    Module.Index1Float1Write=true
    Module.Index1Float2Write=true
    Module.Index1Float3Write=true
    Write Parameter Index 1=::NewEnumerator1
   )
   &SkeletalMeshLocation(
    Bone / Socket Index=$Particles.PartitionParticles.Partition
    Bone / Socket Sampling Mode=::NewEnumerator5
    MaskSpawnGroup=false
    Mesh Sampling Type=::NewEnumerator6
    Orientation Sampling=::NewEnumerator1
    Skeletal Mesh=$System.SkeletalMesh
   ) {module=SkeletalMeshLocation001}
  ]
  rendererClasses [&NiagaraMeshRendererProperties, &NiagaraSpriteRendererProperties]
  renderers [
   {rendererClass=&NiagaraMeshRendererProperties, rendererIndex=0}
   {rendererClass=&NiagaraMeshRendererProperties, rendererIndex=1}
   {rendererClass=&NiagaraSpriteRendererProperties, rendererIndex=2}
  ]
 }
 NE_VC_Boost_Sprite_Fire002 {
  emitterUpdateScript [
   &EmitterState(Inactive Response=::NewEnumerator1, Life Cycle Mode=::NewEnumerator1, UseLoopDelay=false)
   &VC_FilteredSocketCount()
   &SpawnRate(
    SpawnRate=&Multiply_Float(
     A=20
     B=&VC_Mask_Floats(Coro Value:NiagaraFloat=0.5, FrontEnd Value:NiagaraFloat=0.5, FN Value:NiagaraFloat=0.5)
    )
    Use Spawn Probability=false
   )
   &SpawnRate(
    SpawnGroup=1
    SpawnRate=&Multiply_Float(
     A=40
     B=&VC_Mask_Floats(Coro Value:NiagaraFloat=0.5, FrontEnd Value:NiagaraFloat=0.5, FN Value:NiagaraFloat=0.5)
    )
    Use Spawn Probability=false
   ) {module=SpawnRate001}
  ]
  particleSpawnScript [
   /SetVariables(Particles.VisibilityTag=&SpawnGroup()) {isSetParameters=true, module=SetVariables_#4}
   &VC_PaintedVariantsOverride()
   &InitializeParticle(
    Color Mode=::NewEnumerator0
    Lifetime=0.1
    Mesh Scale=&MergeHelper_VectorSwitch(
     0=&RandomRangeVector(Minimum=<0.8 0.8 0.8>, Maximum=<1 1 1>)
     1=&Lerp_Vector(
      A=&RandomRangeVector(Minimum=<1.2 0.6 1.2>, Maximum=<1.4 0.8 1.4>)
      B=&RandomRangeVector(Minimum=<0.4 1 0.4>, Maximum=<0.6 1.2 0.6>)
      Alpha=&MakeFloatFromInt(Int=&RandomRangeInt(Minimum=0, Maximum=1))
     )
     2=?
     3=?
     Index=&SpawnGroup()
     Num=::NewEnumerator0
    )
    Mesh Scale Mode=::NewEnumerator4
    UsePositionOffset=false
    Write Material Random=true
   )
   &PartitionParticles(A=0, Partition Mode:ENiagara_PartitionMode_v2=::NewEnumerator4)
   &SkeletalMeshLocation(
    Bone / Socket Index=$Particles.PartitionParticles.Partition
    Bone / Socket Sampling Mode=::NewEnumerator5
    MaskSpawnGroup=false
    Mesh Sampling Type=::NewEnumerator6
    Orientation Sampling=::NewEnumerator1
    Skeletal Mesh=$System.SkeletalMesh
   )
   &AddVelocity(
    Cone Angle=90
    Cone Axis=<-1 0 0>
    Module.UseConeFalloff=false
    Module.UseVelDistribution=false
    Override Seed=false
    Velocity Mode=::NewEnumerator2
    Velocity Speed=1
   ) {enabled=false}
   &DLR_ProductPreviews(GameModeID:NiagaraInt32=$System.GameModeID, bPreviewVehicleSpeed:NiagaraBool=$System.bPreviewVehicleSpeed)
   &InitialMeshOrientation(
    Mesh Orientation Mode=::NewEnumerator7
    Rotation=&RandomRangeVector(Minimum=<0 0 0>, Maximum=<1 0 0>)
    Use Rotation Vector=true
   )
  ]
  particleUpdateScript [
   &ParticleState()
   &UpdateMeshOrientation(
    Facing Direction=<1 0 -0.1>
    Facing Interpolation=false
    Orientation Method=::NewEnumerator1
    Override Quat=false
    TransformFacing=false
    TransformUp=false
   )
   &InheritSourceMovement(
    Applied Position Scale=&Multiply_VectorByFloat(
     Vector=<1 1 1>
     Float=&FloatFromCurve(
      FloatCurve=DataInterface() {propertyValues {Curve {
         defaultValue=1
         keys [
          {arriveTangent=-2.2077, interpMode=RCIM_Cubic, leaveTangent=-2.2077, tangentMode=RCTM_User, time=0, value=1}
          {tangentMode=RCTM_User, time=1, value=0}
         ]
        }}}
      Scale Curve=0
     )
    )
    ApplyForce=false
    ApplyPosition=true
    ApplyPositionFalloff=false
   ) {enabled=false}
   &SolveForcesAndVelocity(Clamp Velocity=false, Limit Acceleration=false)
   &DynamicMaterialParameters(
    Index 1 Param 2=0.35
    Module.Index0Float0Write=true
    Module.Index0Float1Write=true
    Module.Index0Float2Write=true
    Module.Index0Float3Write=true
    Module.Index1Float0Write=true
    Module.Index1Float1Write=true
    Module.Index1Float2Write=true
    Module.Index1Float3Write=true
    Write Parameter Index 1=::NewEnumerator1
   )
   &SkeletalMeshLocation(
    Bone / Socket Index=$Particles.PartitionParticles.Partition
    Bone / Socket Sampling Mode=::NewEnumerator5
    MaskSpawnGroup=false
    Mesh Sampling Type=::NewEnumerator6
    Orientation Sampling=::NewEnumerator1
    Skeletal Mesh=$System.SkeletalMesh
   ) {module=SkeletalMeshLocation001}
  ]
  rendererClasses [&NiagaraMeshRendererProperties, &NiagaraSpriteRendererProperties]
  renderers [
   {rendererClass=&NiagaraMeshRendererProperties, rendererIndex=0}
   {rendererClass=&NiagaraMeshRendererProperties, rendererIndex=1}
   {rendererClass=&NiagaraSpriteRendererProperties, rendererIndex=2}
  ]
 }
 NE_VC_Boost_Sprite_HaHa {
  emitterUpdateScript [
   &EmitterState()
   &VC_FilteredSocketCount()
   &SpawnRate(
    SpawnRate=&Multiply_Float(
     A=25
     B=&VC_Mask_Floats(Coro Value:NiagaraFloat=1, FrontEnd Value:NiagaraFloat=1, FN Value:NiagaraFloat=0.5)
    )
    Use Spawn Probability=false
   )
   &SpawnRate(
    SpawnGroup=1
    SpawnRate=&Multiply_Float(
     A=8
     B=&VC_Mask_Floats(Coro Value:NiagaraFloat=1, FrontEnd Value:NiagaraFloat=1, FN Value:NiagaraFloat=0.5)
    )
    Use Spawn Probability=false
   ) {module=SpawnRate001}
   &SpawnRate(
    SpawnGroup=2
    SpawnRate=&Multiply_Float(
     A=5
     B=&VC_Mask_Floats(Coro Value:NiagaraFloat=1, FrontEnd Value:NiagaraFloat=1, FN Value:NiagaraFloat=0.5)
    )
    Use Spawn Probability=false
   ) {module=SpawnRate002}
   /SetVariables(Emitter.Base Color - Tint:Vector4f=&MakeVector4_FromLinearColor(LinearColor=&MakeCustomLinearColorFromBool(Bool=$User.bIsFrontendPreview, True Color=<rgba 0.3 0.0188 0.6 1>, False Color=<rgba 0.327 0.0219 0.7 1>))) {isSetParameters=true}
  ]
  particleSpawnScript [
   /SetVariables(Particles.VisibilityTag=&SpawnGroup()) {isSetParameters=true}
   &VC_PaintedVariantsOverride() {enabled=false}
   &InitializeParticle(
    Color=&MakeCustomLinearColorFromBool(Bool=$User.bIsFrontendPreview, True Color=<rgba 0.0386 0.12 0.006 1>, False Color=<rgba 0.0482 0.15 0.0075 1>)
    Lifetime=&MergeHelper_FloatSwitch(
     0=&RandomRangeFloat(Minimum=0.6, Maximum=1)
     1=&RandomRangeFloat(Minimum=0.6, Maximum=0.8)
     2=&RandomRangeFloat(Minimum=0.6, Maximum=0.8)
     3=?
     Index=&SpawnGroup()
     Num=::NewEnumerator1
    )
    Sprite Rotation Angle=&MergeHelper_FloatSwitch(
     0=&RandomRangeFloat(Minimum=-0.1, Maximum=0.1)
     1=&RandomRangeFloat(Minimum=-0.1, Maximum=0.1)
     2=&RandomRangeFloat(Minimum=-0.1, Maximum=0.1)
     3=?
     Index=&SpawnGroup()
     Num=::NewEnumerator1
    )
    Sprite Rotation Mode=::NewEnumerator1
    Sprite Size=&MergeHelper_Vector2Switch(
     0=&MakeCustomVector2DFromBool(Bool=$User.bIsFrontendPreview, True Vector2D=<50 50>, False Vector2D=<75 75>)
     1=&MakeCustomVector2DFromBool(Bool=$User.bIsFrontendPreview, True Vector2D=<100 50>, False Vector2D=<150 75>)
     2=&MakeCustomVector2DFromBool(Bool=$User.bIsFrontendPreview, True Vector2D=<200 50>, False Vector2D=<300 75>)
     3=?
     4=?
     Index=&SpawnGroup()
     Num=::NewEnumerator1
    )
    Sprite Size Mode=::NewEnumerator4
    UsePositionOffset=false
    Write Material Random=true
   )
   &PartitionParticles(Number Of Partitions=$Emitter.PortCount)
   &SkeletalMeshLocation(
    Bone / Socket Index=$Particles.PartitionParticles.Partition
    Bone / Socket Sampling Mode=::NewEnumerator5
    MaskSpawnGroup=false
    Mesh Sampling Type=::NewEnumerator6
    Sampled Position Offset=&RandomRangeVector(Minimum=<-10 -10 -10>, Maximum=<10 10 10>)
    Skeletal Mesh=$System.SkeletalMesh
   )
   &AddVelocity(Velocity=&VC_Mask_Vectors(
     Coro Value:Vector3f=&MergeHelper_VectorSwitch(
      0=&RandomRangeVector(Minimum=<-700 -700 -700>, Maximum=<-400 700 700>)
      1=&RandomRangeVector(Minimum=<-525 -525 -525>, Maximum=<-300 525 525>)
      2=&RandomRangeVector(Minimum=<-525 -525 -525>, Maximum=<-300 525 525>)
      3=?
      Index=&SpawnGroup()
      Num=::NewEnumerator1
     )
     FrontEnd Value:Vector3f=&MergeHelper_VectorSwitch(
      0=&MakeCustomVectorFromBool(
       Bool=$User.bIsFrontendPreview
       True Vector=&RandomRangeVector(Minimum=<-2500 -600 -600>, Maximum=<-2500 800 800>)
       False Vector=&RandomRangeVector(Minimum=<-600 -600 -600>, Maximum=<-400 800 800>)
      )
      1=&MakeCustomVectorFromBool(
       Bool=$User.bIsFrontendPreview
       True Vector=&RandomRangeVector(Minimum=<-1800 -475 -475>, Maximum=<-1800 675 675>)
       False Vector=&RandomRangeVector(Minimum=<-525 -475 -475>, Maximum=<-300 675 675>)
      )
      2=&MakeCustomVectorFromBool(
       Bool=$User.bIsFrontendPreview
       True Vector=&RandomRangeVector(Minimum=<-1500 -475 -475>, Maximum=<-1500 675 675>)
       False Vector=&RandomRangeVector(Minimum=<-525 -475 -475>, Maximum=<-300 675 675>)
      )
      3=?
      Index=&SpawnGroup()
      Num=::NewEnumerator1
     )
     FN Value:Vector3f=&MergeHelper_VectorSwitch(
      0=&RandomRangeVector(Minimum=<-700 -700 -700>, Maximum=<-400 700 700>)
      1=&RandomRangeVector(Minimum=<-525 -525 -525>, Maximum=<-300 525 525>)
      2=&RandomRangeVector(Minimum=<-525 -525 -525>, Maximum=<-300 525 525>)
      3=?
      Index=&SpawnGroup()
      Num=::NewEnumerator1
     )
    ))
   &DLR_ProductPreviews(GameModeID:NiagaraInt32=$System.GameModeID, bPreviewVehicleSpeed:NiagaraBool=$System.bPreviewVehicleSpeed)
  ]
  particleUpdateScript [
   &ParticleState()
   &ScaleSpriteSize(Scale Sprite Size Mode=::NewEnumerator0, Uniform Scale Factor=1.8)
   &InheritSourceMovement(
    Applied Position Scale=&Multiply_VectorByFloat(
     Vector=<1 1 1>
     Float=&FloatFromCurve(
      FloatCurve=DataInterface() {propertyValues {Curve {
         defaultValue=1
         keys [
          {arriveTangent=-2.2077, interpMode=RCIM_Cubic, leaveTangent=-2.2077, tangentMode=RCTM_User, time=0, value=1}
          {tangentMode=RCTM_User, time=1, value=0}
         ]
        }}}
      Scale Curve=0
     )
    )
    ApplyForce=false
    ApplyPosition=true
    ApplyPositionFalloff=false
   )
   &ScaleVelocity(
    Coordinate Space:ENiagaraCoordinateSpace=::Simulation
    Velocity Scale=&MergeHelper_VectorSwitch(
     0=&MakeCustomVectorFromBool(
      Bool=$User.bIsFrontendPreview
      True Vector=&VectorFromFloat(Value=&FloatFromCurve(FloatCurve=DataInterface() {propertyValues {Curve {keys [<at 0 1 RCIM_Cubic>,<at 0.5 0.5 RCIM_Cubic>]}}}, Scale Curve=1.5))
      False Vector=&VectorFromFloat(Value=&FloatFromCurve(FloatCurve=DataInterface() {propertyValues {Curve {keys [<at 0 1 RCIM_Cubic>,<at 0.5 0.2 RCIM_Cubic>]}}}, Scale Curve=1.5))
     )
     1=&MakeCustomVectorFromBool(
      Bool=$User.bIsFrontendPreview
      True Vector=&VectorFromFloat(Value=&FloatFromCurve(FloatCurve=DataInterface() {propertyValues {Curve {keys [<at 0 1 RCIM_Cubic>,<at 0.5 0.4 RCIM_Cubic>]}}}, Scale Curve=1.5))
      False Vector=&VectorFromFloat(Value=&FloatFromCurve(FloatCurve=DataInterface() {propertyValues {Curve {keys [<at 0 1 RCIM_Cubic>,<at 0.5 0.1 RCIM_Cubic>]}}}, Scale Curve=1.5))
     )
     2=&MakeCustomVectorFromBool(
      Bool=$User.bIsFrontendPreview
      True Vector=&VectorFromFloat(Value=&FloatFromCurve(FloatCurve=DataInterface() {propertyValues {Curve {keys [<at 0 1 RCIM_Cubic>,<at 0.5 0.3 RCIM_Cubic>]}}}, Scale Curve=1.5))
      False Vector=&VectorFromFloat(Value=&FloatFromCurve(FloatCurve=DataInterface() {propertyValues {Curve {keys [<at 0 1 RCIM_Cubic>,<at 0.5 0.1 RCIM_Cubic>]}}}, Scale Curve=1.5))
     )
     3=?
     Index=&SpawnGroup()
     Num=::NewEnumerator1
    )
   )
   &SolveForcesAndVelocity(Clamp Velocity=false, Limit Acceleration=false)
   &DynamicMaterialParameters(
    Index 0 Param 2=1
    Index 1 Param 2=0.4
    Module.Index0Float0Write=true
    Module.Index0Float1Write=true
    Module.Index0Float2Write=true
    Module.Index0Float3Write=true
    Module.Index1Float0Write=true
    Module.Index1Float1Write=true
    Module.Index1Float2Write=true
    Module.Index1Float3Write=true
    Write Parameter Index 1=::NewEnumerator1
   )
   /SetVariables(Particles.SubImageIndex=&MergeHelper_FloatSwitch(
     0=&FloatFromCurve(FloatCurve=DataInterface() {propertyValues {Curve {keys [<at 0 0>,<at 1 1>]}}}, Scale Curve=15)
     1=&FloatFromCurve(FloatCurve=DataInterface() {propertyValues {Curve {keys [<at 0 0>,<at 1 1>]}}}, Scale Curve=7)
     2=&FloatFromCurve(FloatCurve=DataInterface() {propertyValues {Curve {keys [<at 0 0>,<at 1 1>]}}}, Scale Curve=3)
     3=?
     Index=&SpawnGroup()
     Num=::NewEnumerator1
    )) {isSetParameters=true}
  ]
  rendererClasses [&NiagaraSpriteRendererProperties]
  renderers [
   {rendererClass=&NiagaraSpriteRendererProperties}
   {rendererClass=&NiagaraSpriteRendererProperties, rendererIndex=1}
   {rendererClass=&NiagaraSpriteRendererProperties, rendererIndex=2}
  ]
 }
}
summary {
 systemName=NS_Clown
 emitters [
  {emitterName=NE_VC_Boost_Sprite_HaHa, rendererClasses [&NiagaraSpriteRendererProperties]}
  {
   emitterName=NE_VC_Boost_Sprite_Fire001
   rendererClasses [&NiagaraMeshRendererProperties, &NiagaraSpriteRendererProperties]
  }
  {
   emitterName=NE_VC_Boost_Sprite_Fire002
   rendererClasses [&NiagaraMeshRendererProperties, &NiagaraSpriteRendererProperties]
  }
 ]
 userVariables [
  User.bIsFrontendPreview {type=NiagaraBool, defaultValue=true}
  User.bIsPainted {type=NiagaraBool}
  User.blsIsCoroColor {type=NiagaraBool}
  User.bPreviewVehicleSpeed {type=NiagaraBool}
  User.ColorVariantID {type=NiagaraInt32}
  User.GameModeID {type=NiagaraInt32}
  User.PaintedColor {type=LinearColor, description="", defaultValue {struct=LinearColor, value=<rgba 1 1 1 1>}}
  User.PaintedScalar {type=NiagaraFloat, defaultValue=1, description=""}
  User.PortCount {type=NiagaraInt32, defaultValue=2}
 ]
}
